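# Niagara effect for the SeleneCobra glider's deploy, with a cyan-blue to purple glow.
# Mesh emitters draw scaling wing, spinning ring and trail burst shapes, fading over their lifetimes.
# Two sprite emitters add flickering particles, one burst and one swirled by curl noise and vortex, while partition-indexed arrays place two mirrored lights.
# User.bIsFrontEndPreview dims several emitters and adjusts some lifetimes, offsets and vortex strength for front-end preview.
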NiagaraSystem /BRCosmetics/Gadgets/Assets/VinderTech_GliderChute/Glider_SeleneCobra/FX/NS_SeleneCobra_Glider_Deploy
dependencies {
 usedDataInterfaces [&NiagaraDataInterfaceCurve, &NiagaraDataInterfaceColorCurve, &NiagaraDataInterfaceArrayFloat3]
 usedDynamicInputs []
 usedModules [&SystemState, &InterpolateOverTime]
}
emitters {
 Light {
  emitterUpdateScript [
   &EmitterState(Life Cycle Mode=::NewEnumerator1, Loop Behavior=::NewEnumerator1, Loop Delay=1, Loop Duration=2, UseLoopDelay=true)
   &SpawnBurst_Instantaneous(Spawn Count=2, Use Spawn Probability=false, UseLoopCountLimit=false)
  ]
  particleSpawnScript [
   &PartitionParticles()
   &InitializeParticle(
    Color Mode=::NewEnumerator0
    Lifetime=0.8
    Position Offset=&SelectVectorFromArray(
     Direct Array Index=$Particles.PartitionParticles.Partition
     Vector Selection Array=DataInterface() {propertyValues {FloatData [<90 50 -5>, <-90 50 -5>]}}
    )
    UsePositionOffset=true
    Write Material Random=true
   )
  ]
  particleUpdateScript [
   &ParticleState()
   &FN_Light_Attributes(
    Color=&ColorFromCurve(ColorCurve=DataInterface() {propertyValues {
       AlphaCurve {keys []}
       BlueCurve {keys [<at 0 0 RCIM_Cubic>,<at 0.1507 5>,<at 0.5355 3>,<at 1 0 RCIM_Cubic>]}
       GreenCurve {keys [<at 0 0 RCIM_Cubic>,<at 0.1507 1.25>,<at 0.5355 0.1208>,<at 1 0 RCIM_Cubic>]}
       RedCurve {keys [<at 0 0 RCIM_Cubic>,<at 0.1507 0>,<at 0.5355 0.3256>,<at 1 0 RCIM_Cubic>]}
      }})
    Create Color Attribute=true
    Create Exponent Attribute=true
    Create Position Attribute=true
    Create Radius Attribute=true
    Exponent=1.5
    Position=&SelectVectorFromArray(
     Direct Array Index=$Particles.PartitionParticles.Partition
     Vector Selection Array=DataInterface() {propertyValues {FloatData [<90 50 -5>, <-90 50 -5>]}}
    )
    Radius=80
    Scale Color=80
   )
  ]
  rendererClasses [&NiagaraLightRendererProperties]
  renderers [{rendererClass=&NiagaraLightRendererProperties, rendererIndex=0}]
 }
 Particle_Burst {
  emitterUpdateScript [
   &EmitterState(Life Cycle Mode=::NewEnumerator1, Loop Behavior=::NewEnumerator1, Loop Delay=1, Loop Duration=2, UseLoopDelay=true)
   &SpawnBurst_Instantaneous(Spawn Count=30, Use Spawn Probability=false, UseLoopCountLimit=false)
  ]
  particleSpawnScript [
   &InitializeParticle(
    Color Mode=::NewEnumerator0
    Lifetime Max=1
    Lifetime Min=0.3
    Lifetime Mode=::NewEnumerator1
    Sprite Size Mode=::NewEnumerator1
    Uniform Sprite Size Max=50
    Uniform Sprite Size Min=30
    UsePositionOffset=false
    Write Material Random=true
   )
   &ShapeLocation(
    MaskSpawnGroup=false
    Non Uniform Scale=<1 0.5 0.2>
    Override Seed=false
    Rotation Axis=<0 1 0>
    Rotation Mode=::NewEnumerator1
    Shape Origin=$Engine.Emitter.SimulationPosition
    Sphere Radius=120
    UseOwnerScale=true
   ) {module=ShapeLocation001}
   &AddVelocity(Module.Use Falloff Distance=false, Override Seed=false, Velocity Mode=::NewEnumerator1, Velocity Speed=50) {module=AddVelocity001}
   &AddVelocity(Velocity=&RandomRangeVector(Minimum=<0 400 0>, Maximum=<0 600 0>))
  ]
  particleUpdateScript [
   &ParticleState()
   &ScaleSpriteSize(Uniform Curve Sprite Scale=DataInterface() {propertyValues {Curve {keys [<at 0 0>, {arriveTangent=1, time=0.164, value=0.984}, <at 0.4045 0.258>, <at 0.4964 0.596>, {leaveTangent=-2.9495, time=1, value=0}]}}})
   &ScaleColor(
    Scale Alpha=&FloatFromCurve(
     FloatCurve=DataInterface() {propertyValues {Curve {keys [<at 0 0.2>,<at 0.1 0.85>,<at 0.2 0.15>,<at 0.4 0.8>,<at 0.5 0.4>,<at 0.7 0.45>,<at 1 0>]}}}
     CurveIndex=$Particles.Age
     Scale Curve=&MakeCustomFloatFromBool(Bool=$User.bIsFrontEndPreview, True Float=0.6, False Float=1.2)
    )
    Scale RGB=&MakeVectorFromLinearColorRGB(Color=&ColorFromCurve(
      ColorCurve=DataInterface() {propertyValues {
        AlphaCurve {keys [<at 0 1 RCIM_Cubic>]}
        BlueCurve {keys [<at 0 1.5 RCIM_Cubic>,<at 0.2296 1.5>,<at 0.4498 1>,<at 0.7443 1 RCIM_Cubic>]}
        GreenCurve {keys [<at 0 1.01 RCIM_Cubic>,<at 0.2296 0.03>,<at 0.4498 0>,<at 0.7443 0.0208 RCIM_Cubic>]}
        RedCurve {keys [<at 0 0.03 RCIM_Cubic>,<at 0.2296 0.03>,<at 0.4498 0.1667>,<at 0.7443 0.4 RCIM_Cubic>]}
       }}
      CurveIndex=$Particles.Age
     ))
    ScaleA=true
    ScaleRGB=true
   )
   /SetVariables(Particles.Color002=&MultiplyLinearColor(Linear Color=<rgba 1 1 1 0.2>, Scale Factor=&MakeVector4_FromLinearColor(LinearColor=$Particles.Color))) {isSetParameters=true}
   &SolveForcesAndVelocity(Clamp Velocity=false, Limit Acceleration=false)
  ]
  rendererClasses [&NiagaraSpriteRendererProperties]
  renderers [
   {rendererClass=&NiagaraSpriteRendererProperties}
   {rendererClass=&NiagaraSpriteRendererProperties, rendererIndex=1}
  ]
 }
 Particle_Once {
  emitterUpdateScript [
   &EmitterState(Life Cycle Mode=::NewEnumerator1, Loop Behavior=::NewEnumerator1, Loop Duration=0.5, UseLoopDelay=false)
   &SpawnRate(SpawnRate=150, Use Spawn Probability=false)
  ]
  particleSpawnScript [
   &PartitionParticles()
   &InitializeParticle(
    Color Mode=::NewEnumerator0
    Lifetime Max=&MakeCustomFloatFromBool(Bool=$User.bIsFrontEndPreview, True Float=0.8, False Float=1.5)
    Lifetime Min=0.5
    Lifetime Mode=::NewEnumerator1
    Sprite Size Mode=::NewEnumerator1
    Uniform Sprite Size Max=40
    Uniform Sprite Size Min=20
    UsePositionOffset=false
    Write Material Random=true
   )
   &ShapeLocation(
    Handle Radius=60
    Large Radius=90
    MaskSpawnGroup=false
    Override Seed=false
    Rotation Axis=<0 1 0>
    Rotation Mode=::NewEnumerator1
    Shape Origin=$Engine.Emitter.SimulationPosition
    Shape Primitive=::NewEnumerator3
    UseOwnerScale=true
   ) {module=ShapeLocation001}
   &AddVelocity(Module.Use Falloff Distance=false, Override Seed=false, Velocity Mode=::NewEnumerator1, Velocity Speed=15)
  ]
  particleUpdateScript [
   &ParticleState()
   &CurlNoiseForce(Noise Frequency=10, Noise Strength=600, Pan Noise=false)
   &VortexForce(
    UseOriginOffset=false
    Vortex Axis=<0 1 0>
    Vortex Force Amount=&MakeCustomFloatFromBool(Bool=$User.bIsFrontEndPreview, True Float=-300, False Float=-500)
   )
   &Drag(
    Drag=&FloatFromCurve(FloatCurve=DataInterface() {propertyValues {Curve {keys [<at 0.5 0>,<at 1 5>]}}})
    Use Linear Drag=true
    Use Rotational Drag=true
   )
   &ScaleSpriteSize(Uniform Curve Sprite Scale=DataInterface() {propertyValues {Curve {keys [<at 0 0>, {arriveTangent=1, leaveTangent=1, time=0.2, value=0.825}, <at 0.5 0.225>, <at 0.7 1>, {leaveTangent=-2.9495, time=1, value=0}]}}})
   &ScaleColor(
    Scale Alpha=&FloatFromCurve(
     FloatCurve=DataInterface() {propertyValues {Curve {keys [<at 0 0.2>,<at 0.149 0.9024>,<at 0.2861 0.1555>,<at 0.4024 1.1104>,<at 0.5048 0.2>,<at 0.6364 1.0933>,<at 0.7577 0.125>,<at 0.853 0.8>,<at 1 0>]}}}
     Scale Curve=&MakeCustomFloatFromBool(Bool=$User.bIsFrontEndPreview, True Float=0.4, False Float=1)
    )
    Scale RGB=&MakeVectorFromLinearColorRGB(Color=&ColorFromCurve(ColorCurve=DataInterface() {propertyValues {
        AlphaCurve {keys [<at 0 1 RCIM_Cubic>]}
        BlueCurve {keys [<at 0 5 RCIM_Cubic>,<at 0.2296 1>,<at 0.5592 1>,<at 0.8874 1 RCIM_Cubic>]}
        GreenCurve {keys [<at 0 3.3667 RCIM_Cubic>,<at 0.2296 0.02>,<at 0.5592 0>,<at 0.8874 0 RCIM_Cubic>]}
        RedCurve {keys [<at 0 0.1 RCIM_Cubic>,<at 0.2296 0.02>,<at 0.5592 0.1667>,<at 0.8874 0.6667 RCIM_Cubic>]}
       }}))
    ScaleA=true
    ScaleRGB=true
   )
   /SetVariables(Particles.Color002=&MultiplyLinearColor(Linear Color=<rgba 1 1 1 0.15>, Scale Factor=&MakeVector4_FromLinearColor(LinearColor=$Particles.Color))) {isSetParameters=true}
   &SolveForcesAndVelocity(Clamp Velocity=false, Limit Acceleration=false)
  ]
  rendererClasses [&NiagaraSpriteRendererProperties]
  renderers [
   {rendererClass=&NiagaraSpriteRendererProperties}
   {rendererClass=&NiagaraSpriteRendererProperties, rendererIndex=1}
  ]
 }
 Trail_Burst {
  emitterUpdateScript [
   &EmitterState(
    Inactive Response=::NewEnumerator1
    Life Cycle Mode=::NewEnumerator1
    Loop Behavior=::NewEnumerator1
    Loop Delay=1
    Loop Duration=2
    UseLoopDelay=true
   )
   &SpawnBurst_Instantaneous(Use Spawn Probability=false, UseLoopCountLimit=false)
  ]
  particleSpawnScript [
   &InitializeParticle(Mesh Scale=<1.02 0.9 1>, Mesh Scale Mode=::NewEnumerator4, Position Offset=<0 27 -5>, UsePositionOffset=true, Write Material Random=true)
   /SetVariables(Particles.Size002:Vector3f=&Multiply_Vector(A=<1 0.9 1>, B=&VectorFromFloat(Value=2.45))) {isSetParameters=true}
   /SetVariables(Particles.Mesh Orientation002:Quat4f=<0 180 0 1>) {isSetParameters=true, module=SetVariables_#1}
   /SetVariables(Particles.Position002:NiagaraPosition=<0 35 -5>) {isSetParameters=true}
   /SetVariables(Particles.Position003:NiagaraPosition=<0 0 -5>) {isSetParameters=true}
   &CameraOffset(Camera Offset Amount=-10)
  ]
  particleUpdateScript [
   &ParticleState()
   &ScaleMeshSize(Scale Factor=&VectorFromFloat(Value=2.45))
   &DynamicMaterialParameters(
    Index 0 Param 1=&RandomRangeFloat(Minimum=0, Maximum=1)
    Index 0 Param 2=&FloatFromCurve(FloatCurve=DataInterface() {propertyValues {Curve {keys [
         <at 0 0>
         {arriveTangent=-0.6282, interpMode=RCIM_Cubic, leaveTangent=-0.6282, tangentMode=RCTM_User, time=0.2, value=1}
         <at 1 0.8 RCIM_Cubic>
        ]}}})
    Module.Index0Param0Write=true
    Module.Index0Param1Write=true
    Module.Index0Param2Write=true
    Module.Index0Param3Write=true
   )
   &ScaleColor(
    Scale Alpha=&FloatFromCurve(FloatCurve=DataInterface() {propertyValues {Curve {keys [
         {interpMode=RCIM_Cubic, leaveTangent=-11.1881, tangentMode=RCTM_Break, time=0, value=1}
         {arriveTangent=-1.3671, interpMode=RCIM_Cubic, leaveTangent=-1.3671, tangentMode=RCTM_User, time=0.0956, value=0.4244}
         <at 1 0>
        ]}}})
    Scale RGB=<2.5 2.5 2.5>
    ScaleA=true
    ScaleRGB=true
   )
   /SetVariables(Particles.Color002=&MakeLinearColorFromVectorAndFloat(
     "Vector (RGB)"=<1 1 1>
     "Float (Alpha)"=&FloatFromCurve(
      FloatCurve=DataInterface() {propertyValues {Curve {keys [
          {interpMode=RCIM_Cubic, leaveTangent=-6.4008, tangentMode=RCTM_Break, time=0, value=1}
          {arriveTangent=-1.0229, interpMode=RCIM_Cubic, leaveTangent=-1.0229, tangentMode=RCTM_User, time=0.2, value=0.25}
          <at 1 0 RCIM_Cubic>
         ]}}}
      Scale Curve=0.6
     )
    )) {isSetParameters=true}
   /SetVariables(Particles.Color003:LinearColor=&MakeLinearColorFromVectorAndFloat(
     "Vector (RGB)"=<5 5 5>
     "Float (Alpha)"=&FloatFromCurve(
      FloatCurve=DataInterface() {propertyValues {Curve {keys [
          {interpMode=RCIM_Cubic, leaveTangent=-10.3154, tangentMode=RCTM_Break, time=0, value=1}
          {arriveTangent=-0.8269, interpMode=RCIM_Cubic, leaveTangent=-0.8269, tangentMode=RCTM_User, time=0.1251, value=0.2621}
          {interpMode=RCIM_Cubic, tangentMode=RCTM_User, time=1, value=0}
         ]}}}
      Scale Curve=1.5
     )
    )) {isSetParameters=true}
   /SetVariables(Particles.Color004:LinearColor=&MakeLinearColorFromVectorAndFloat(
     "Vector (RGB)"=&MakeVectorFromLinearColorRGB(Color=<rgba 5 5 5 1>)
     "Float (Alpha)"=&FloatFromCurve(FloatCurve=DataInterface() {propertyValues {Curve {keys [
          {interpMode=RCIM_Cubic, leaveTangent=-10.3154, tangentMode=RCTM_Break, time=0, value=1}
          {arriveTangent=-0.8269, interpMode=RCIM_Cubic, leaveTangent=-0.8269, tangentMode=RCTM_User, time=0.1251, value=0.2621}
          {interpMode=RCIM_Cubic, tangentMode=RCTM_User, time=1, value=0}
         ]}}})
    )) {isSetParameters=true}
   &ScaleMeshSize(Scale Factor=&MakeVector(
     X=1
     Y=&FloatFromCurve(FloatCurve=DataInterface() {propertyValues {Curve {keys [
          {arriveTangent=0.0132, interpMode=RCIM_Cubic, leaveTangent=0.0132, tangentMode=RCTM_User, time=-0.0021, value=0.9005}
          {arriveTangent=0.2288, interpMode=RCIM_Cubic, leaveTangent=0.2288, time=0.1873, value=1.0948}
          {arriveTangent=-0.2397, interpMode=RCIM_Cubic, leaveTangent=-0.2397, time=0.4128, value=0.9955}
          <at 1 0.9 RCIM_Cubic>
         ]}}})
     Z=1
    )) {module=ScaleMeshSize002}
   /SetVariables(Particles.Size003:Vector3f=&MakeVector(
     X=&FloatFromCurve(
      FloatCurve=DataInterface() {propertyValues {Curve {keys [{arriveTangent=3.0792, interpMode=RCIM_Cubic, leaveTangent=3.0792, tangentMode=RCTM_User, time=0, value=0.8}, <at 0.8 2 RCIM_Cubic>]}}}
      Scale Curve=2.2
     )
     Y=&FloatFromCurve(FloatCurve=DataInterface() {propertyValues {Curve {keys [<at 0 0.5>,<at 0.3 1.3>,<at 1 1.2>]}}}, Scale Curve=2.45)
     Z=&FloatFromCurve(FloatCurve=DataInterface() {propertyValues {Curve {keys [<at 0 1>,<at 1 0.8>]}}}, Scale Curve=2.45)
    )) {isSetParameters=true}
   /SetVariables(Particles.Size004:Vector3f=&MakeVector(
     X=2.45
     Y=&FloatFromCurve(FloatCurve=DataInterface() {propertyValues {Curve {keys [<at 0 1>,<at 0.1 1.3>,<at 0.3 1.15>,<at 1 1>]}}}, Scale Curve=2)
     Z=2
    )) {isSetParameters=true}
  ]
  rendererClasses [&NiagaraMeshRendererProperties]
  renderers [
   {rendererClass=&NiagaraMeshRendererProperties, rendererIndex=0}
   {rendererClass=&NiagaraMeshRendererProperties, rendererIndex=1}
   {rendererClass=&NiagaraMeshRendererProperties, rendererIndex=2}
   {rendererClass=&NiagaraMeshRendererProperties, rendererIndex=3}
  ]
 }
 Trail_Ring {
  emitterUpdateScript [
   &EmitterState()
   &SpawnBurst_Instantaneous(Use Spawn Probability=false, UseLoopCountLimit=false)
  ]
  particleSpawnScript [
   &InitializeParticle(Lifetime=0.7, Mesh Scale=<2 2 2>, Mesh Scale Mode=::NewEnumerator4, UsePositionOffset=true, Write Material Random=true)
   &InitialMeshOrientation(Mesh Orientation Mode=::NewEnumerator2, Rotation=<0 0.25 0>, Use Rotation Vector=true)
  ]
  particleUpdateScript [
   &ParticleState()
   &UpdateMeshOrientation(Rotation Rate=3, Rotation Vector=<0 0 -1>)
   &ScaleMeshSize(Scale Factor=&VectorFromFloat(Value=&FloatFromCurve(
      FloatCurve=DataInterface() {propertyValues {Curve {keys [
          {arriveTangent=2.1412, interpMode=RCIM_Cubic, leaveTangent=2.1412, tangentMode=RCTM_User, time=0, value=0.6}
          {arriveTangent=0.1501, interpMode=RCIM_Cubic, leaveTangent=0.1501, tangentMode=RCTM_User, time=0.6, value=1.4}
         ]}}}
      Scale Curve=1.8
     )))
   &ScaleColor(
    Scale Alpha=&FloatFromCurve(
     FloatCurve=DataInterface() {propertyValues {Curve {keys [
         {arriveTangent=8.4005, interpMode=RCIM_Cubic, time=0, value=0}
         {arriveTangent=0.0321, interpMode=RCIM_Cubic, leaveTangent=0.0321, tangentMode=RCTM_User, time=0.1128, value=1.0035}
         {arriveTangent=-0.8711, interpMode=RCIM_Cubic, leaveTangent=-0.8711, tangentMode=RCTM_User, time=0.6266, value=0.7899}
         {interpMode=RCIM_Cubic, leaveTangent=0.024, time=1, value=0}
        ]}}}
     Scale Curve=&MakeCustomFloatFromBool(Bool=$User.bIsFrontEndPreview, True Float=0.2, False Float=1)
    )
    ScaleA=true
    ScaleRGB=true
   )
  ]
  rendererClasses [&NiagaraMeshRendererProperties]
  renderers [{rendererClass=&NiagaraMeshRendererProperties, rendererIndex=0}]
 }
 Trail_wing {
  emitterUpdateScript [
   &EmitterState()
   &SpawnBurst_Instantaneous(Use Spawn Probability=false, UseLoopCountLimit=false)
  ]
  particleSpawnScript [
   &InitializeParticle(
    Lifetime=&MakeCustomFloatFromBool(Bool=$User.bIsFrontEndPreview, True Float=0.95, False Float=0.8)
    Mesh Scale=<2 2 2>
    Mesh Scale Mode=::NewEnumerator4
    Position Offset=&MakeCustomVectorFromBool(Bool=$User.bIsFrontEndPreview, True Vector=<0 -10 0>, False Vector=<0 0 0>)
    UsePositionOffset=true
    Write Material Random=true
   )
   &InitialMeshOrientation(Mesh Orientation Mode=::NewEnumerator2, Rotation=<-0.25 0 0>, Use Rotation Vector=true)
   /SetVariables(Particles.Mesh Orientation002:Quat4f=<0 180 180 1>) {isSetParameters=true, module=SetVariables_#1}
   &CameraOffset(Camera Offset Amount=-10)
  ]
  particleUpdateScript [
   &ParticleState()
   &ScaleMeshSize(Scale Factor=&VectorFromFloat(Value=&FloatFromCurve(FloatCurve=DataInterface() {propertyValues {Curve {keys [<at 0 0.55>,<at 0.6 1.15>]}}}, Scale Curve=1.8)))
   &ScaleColor(
    Scale Alpha=&FloatFromCurve(
     FloatCurve=DataInterface() {propertyValues {Curve {keys [
         {arriveTangent=8.4005, interpMode=RCIM_Cubic, time=0, value=0}
         {arriveTangent=0.0321, interpMode=RCIM_Cubic, leaveTangent=0.0321, tangentMode=RCTM_User, time=0.1128, value=1.0035}
         {arriveTangent=-1.1311, interpMode=RCIM_Cubic, leaveTangent=-1.1311, time=0.5469, value=0.7697}
         {interpMode=RCIM_Cubic, leaveTangent=0.024, time=1, value=0}
        ]}}}
     Scale Curve=&MakeCustomFloatFromBool(Bool=$User.bIsFrontEndPreview, True Float=0.4, False Float=1)
    )
    ScaleA=true
    ScaleRGB=true
   )
  ]
  rendererClasses [&NiagaraMeshRendererProperties]
  renderers [
   {rendererClass=&NiagaraMeshRendererProperties, rendererIndex=0}
   {rendererClass=&NiagaraMeshRendererProperties, rendererIndex=1}
  ]
 }
}
summary {
 systemName=NS_SeleneCobra_Glider_Deploy
 emitters [
  {emitterName=Trail_wing, rendererClasses [&NiagaraMeshRendererProperties]}
  {emitterName=Trail_Ring, rendererClasses [&NiagaraMeshRendererProperties]}
  {emitterName=Particle_Once, rendererClasses [&NiagaraSpriteRendererProperties]}
  {emitterName=Trail_Burst, rendererClasses [&NiagaraMeshRendererProperties]}
  {emitterName=Particle_Burst, rendererClasses [&NiagaraSpriteRendererProperties]}
  {emitterName=Light, rendererClasses [&NiagaraLightRendererProperties]}
 ]
 userVariables [
  User.bIsFrontEnd {type=NiagaraBool}
  User.bIsFrontEndPreview {type=NiagaraBool}
  User.ForwardDot {type=NiagaraFloat, defaultValue=0, description=""}
 ]
}
